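# Physics asset for the Moss CarbideBlack backpack cosmetic, defining bodies and joint limits for its dangling moss, vines and leaves.
# A large capsule covers the pack bone, and small capsules sit on each dynamic joint.
# Three moss joints and three vine joints constrain directly to the pack, while five leaves branch from the vines as short two-level chains.
# Constraints set per-joint swing limits with very tight twist, fully locked on the last leaf.
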PhysicsAsset /BRCosmetics/Accessories/FORT_Backpacks/Backpack_M_MED_Moss_CarbideBlack/Meshes/M_MED_Moss_CarbideBlack_Pack_Physics
bodies {
 dyn_hangingMoss_jnt_01 [{shapeType=Capsule, length=2, radius=2.25, center=<1.879 0 0>, rotation=<rot 90 0 0>}]
 dyn_hangingMoss_jnt_02 [{shapeType=Capsule, length=2, radius=2, center=<2.2087 0 0>, rotation=<rot 90 0 0>}]
 dyn_hangingMoss_jnt_03 [{shapeType=Capsule, length=1.75, radius=1.25, center=<1.3958 -0.0001 0>, rotation=<rot 90 0 0>}]
 dyn_leafVine_1_jnt [{shapeType=Capsule, length=1, radius=1, center=<0.6364 0.086 0>, rotation=<rot 90 172.875 152.8745>}]
 dyn_leafVine_2_jnt [{shapeType=Capsule, length=1, radius=0.75, center=<0.5519 0 0>, rotation=<rot 90 0 0>}]
 dyn_leafVine_3_jnt [{shapeType=Capsule, length=0.5, radius=1, center=<0.9174 0 0>, rotation=<rot 90 0 0>}]
 dyn_leaf_jnt_01 [{shapeType=Capsule, length=1, radius=1, center=<0.8205 0 0>, rotation=<rot 90 0 0>}]
 dyn_leaf_jnt_02 [{shapeType=Capsule, length=1, radius=1, center=<1.1896 0 -0.5>, rotation=<rot 90 0 0>}]
 dyn_leaf_jnt_03 [{shapeType=Capsule, length=1, radius=1.25, center=<0.9467 0 -0.5>, rotation=<rot 90 0 0>}]
 dyn_leaf_jnt_04 [{shapeType=Capsule, length=1, radius=1.25, center=<1.2427 0 -0.2395>, rotation=<rot 90 0 0>}]
 dyn_leaf_jnt_05 [{shapeType=Capsule, length=1, radius=1, center=<1.0631 0 0>, rotation=<rot 90 0 0>}]
 pack [{shapeType=Capsule, length=1, radius=12.5}]
}
constraints [
 {bone1Name=dyn_hangingMoss_jnt_01, bone2Name=pack, swing1LimitDegrees=40, swing2LimitDegrees=40, twistLimitDegrees=3}
 {bone1Name=dyn_hangingMoss_jnt_03, bone2Name=pack, swing1LimitDegrees=30, swing2LimitDegrees=30, twistLimitDegrees=3}
 {bone1Name=dyn_hangingMoss_jnt_02, bone2Name=pack, swing1LimitDegrees=60, swing2LimitDegrees=60, twistLimitDegrees=3}
 {bone1Name=dyn_leafVine_1_jnt, bone2Name=pack, swing1LimitDegrees=20, swing2LimitDegrees=20, twistLimitDegrees=3}
 {bone1Name=dyn_leaf_jnt_01, bone2Name=dyn_leafVine_1_jnt, swing1LimitDegrees=30, swing2LimitDegrees=30, twistLimitDegrees=3}
 {bone1Name=dyn_leafVine_2_jnt, bone2Name=pack, swing1LimitDegrees=15, swing2LimitDegrees=15, twistLimitDegrees=3}
 {bone1Name=dyn_leaf_jnt_02, bone2Name=dyn_leafVine_2_jnt, swing1LimitDegrees=30, swing2LimitDegrees=30, twistLimitDegrees=3}
 {bone1Name=dyn_leaf_jnt_03, bone2Name=dyn_leafVine_2_jnt, swing1LimitDegrees=30, swing2LimitDegrees=30, twistLimitDegrees=3}
 {bone1Name=dyn_leafVine_3_jnt, bone2Name=pack, swing1LimitDegrees=20, swing2LimitDegrees=40, twistLimitDegrees=3}
 {bone1Name=dyn_leaf_jnt_04, bone2Name=dyn_leafVine_3_jnt, swing1LimitDegrees=20, swing2LimitDegrees=60, twistLimitDegrees=3}
 {
  bone1Name=dyn_leaf_jnt_05
  bone2Name=dyn_leafVine_3_jnt
  swing1LimitDegrees=15
  swing2LimitDegrees=30
  twistLimitDegrees=3
  twistMotion=Locked
 }
]
